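# Looping cloud effect for the Heaven guitar emote, a Niagara system of swirling smoke, air flow and floating rock meshes.
# NE_Smokes spawns GPU sprites in two rings under a vortex force, with position inheritance that switches on owner speed.
# NE_AirFlow and NE_Vortex add spinning meshes, while two rock emitters swirl random meshes, one rising, one dropping late in life.
# User.Scale sizes most mesh emitters, and time-of-day and front-end helpers adjust color and material parameters.
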NiagaraSystem /SparksCosmetics/Emotes/Guitar/Heaven/Effects/System/NS_Emote_Heaven_CloudsLoop
dependencies {
 usedDataInterfaces [&NiagaraDataInterfaceCurve, &NiagaraDataInterfaceCamera]
 usedDynamicInputs [&Multiply_Vector, &ReturnCameraVector]
 usedModules [&SystemState]
}
emitters {
 NE_AirFlow {
  emitterUpdateScript [
   &EmitterState(Life Cycle Mode=::NewEnumerator1, Loop Behavior=::NewEnumerator1, Loop Duration=10.5, UseLoopDelay=false)
   &SpawnRate(SpawnRate=5, Use Spawn Probability=false) {module=SpawnRate001}
  ]
  particleSpawnScript [
   &InitializeParticle(
    Color=&FNTimeOfDayColor(Night Color=<rgba 0.5517 0.7251 1 0.7>, Morning Color=<rgba 1 1 1 0.5>, Day Color=<rgba 1 1 1 1>, Evening Color=<rgba 1 0.8396 0.7176 1>)
    Lifetime Max=2.5
    Lifetime Min=1.5
    Lifetime Mode=::NewEnumerator1
    Mesh Scale=&Multiply_VectorByFloat(Vector=<1 1 1.2>, Float=&RandomRangeFloat(Minimum=1, Maximum=2))
    Mesh Scale Mode=::NewEnumerator4
    Position Offset=<0 0 -5>
    UsePositionOffset=true
    Write Material Random=true
   )
   &InitialMeshOrientation(
    Mesh Orientation Mode=::NewEnumerator7
    Rotation=&RandomRangeVector(Minimum=<0 0 -1>, Maximum=<0 0 1>)
    Use Rotation Vector=true
   )
   &ShapeLocation(
    Cylinder Height=50
    Cylinder Height Midpoint=0
    Cylinder Radius=0
    MaskSpawnGroup=false
    Non Uniform Scale=&Multiply_VectorByFloat(Vector=<1 1 1>, Float=$User.Scale)
    Override Seed=false
    Shape Primitive=::NewEnumerator1
    Surface Only=false
    UseOwnerScale=true
   )
  ]
  particleUpdateScript [
   &ParticleState()
   &DynamicMaterialParameters(
    Index 0 Param 1=&FloatFromCurve(FloatCurve=DataInterface() {propertyValues {Curve {keys [{interpMode=RCIM_Cubic, tangentMode=RCTM_User, time=0.5, value=0}, {interpMode=RCIM_Cubic, tangentMode=RCTM_User, time=1, value=1}]}}})
    Index 0 Param 2=&FNTimeOfDayScalar(Night Scale=0, Morning Scale=1, Evening Scale=0)
    Index 0 Param 3=1
    Module.Param0WriteEnabled=true
    Module.Param1WriteEnabled=true
    Module.Param2WriteEnabled=true
    Module.Param3WriteEnabled=true
   )
   &ScaleMeshSize(Scale Factor=&Multiply_VectorByFloat(
     Vector=<1 1 1.5>
     Float=&FloatFromCurve(
      FloatCurve=DataInterface() {propertyValues {Curve {keys [
          {arriveTangent=4.1785, interpMode=RCIM_Cubic, leaveTangent=4.1785, tangentMode=RCTM_User, time=0, value=0}
          {arriveTangent=1.1179, interpMode=RCIM_Cubic, leaveTangent=1.1179, tangentMode=RCTM_User, time=0.2042, value=0.6228}
          {arriveTangent=0.0224, interpMode=RCIM_Cubic, leaveTangent=0.0224, tangentMode=RCTM_User, time=1, value=1}
         ]}}}
      Scale Curve=&Multiply_Float(A=1, B=$User.Scale)
     )
    ))
   &UpdateMeshOrientation(
    Rotation Rate=&FloatFromCurve(FloatCurve=DataInterface() {propertyValues {Curve {keys [<at 0 0>,<at 1 1>]}}}, Scale Curve=0.5)
    Rotation Vector=<0 0 1>
   )
   &ScaleColor(
    Scale Alpha=&FloatFromCurve(
     FloatCurve=DataInterface() {propertyValues {Curve {
        defaultValue=0
        keys [
         {interpMode=RCIM_Cubic, tangentMode=RCTM_User, time=0, value=0}
         {interpMode=RCIM_Cubic, tangentMode=RCTM_User, time=0.5, value=0.6}
         {interpMode=RCIM_Cubic, tangentMode=RCTM_User, time=1, value=0}
        ]
       }}}
     Scale Curve=0.5
    )
    ScaleA=true
    ScaleRGB=true
   )
  ]
  rendererClasses [&NiagaraMeshRendererProperties]
  renderers [{rendererClass=&NiagaraMeshRendererProperties, rendererIndex=0}]
 }
 NE_Rocks {
  emitterUpdateScript [
   &EmitterState(Life Cycle Mode=::NewEnumerator1, Loop Behavior=::NewEnumerator1, Loop Delay=0.5, UseLoopDelay=true)
   &SpawnRate(SpawnRate=3, Use Spawn Probability=false) {module=SpawnRate001}
  ]
  particleSpawnScript [
   &InitializeParticle(
    Lifetime=4
    Mesh Scale=<0.005 0.005 0.01>
    Mesh Scale Mode=::NewEnumerator4
    Position Offset=<0 0 5>
    Ribbon Width Mode=::NewEnumerator1
    UsePositionOffset=true
    Write Material Random=true
   )
   &InitialMeshOrientation(
    Mesh Orientation Mode=::NewEnumerator7
    Rotation=&RandomRangeVector(Minimum=<0 0 -1>, Maximum=<0 0 1>)
    Use Rotation Vector=true
   )
   &ShapeLocation(
    Handle Radius=0
    Large Radius=80
    MaskSpawnGroup=false
    Non Uniform Scale=&Multiply_VectorByFloat(Vector=<1 1 1>, Float=$User.Scale)
    Override Seed=false
    Shape Primitive=::NewEnumerator3
    UseOwnerScale=true
   )
   /SetVariables(Particles.MeshIndex=&RandomRangeInt(Minimum=0, Maximum=5)) {isSetParameters=true}
  ]
  particleUpdateScript [
   &ParticleState()
   &ScaleMeshSize(Scale Factor=&VectorFromFloat(Value=&FloatFromCurve(FloatCurve=DataInterface() {propertyValues {Curve {keys [
          {interpMode=RCIM_Cubic, tangentMode=RCTM_User, time=0, value=0}
          {arriveTangent=-2.5809, interpMode=RCIM_Cubic, leaveTangent=-2.5809, tangentMode=RCTM_User, time=0.5, value=1}
          {interpMode=RCIM_Cubic, tangentMode=RCTM_User, time=1, value=0}
         ]}}})))
   &VortexForce(
    Origin Pull Amount=150
    UseOriginOffset=false
    Vortex Axis Coordinate Space:ENiagaraCoordinateSpace=::Local
    Vortex Force Amount=&RandomRangeFloat(
     Minimum=&FloatFromCurve(FloatCurve=DataInterface() {propertyValues {Curve {keys [<at 0 1>,<at 0.5 1>,<at 1 0>]}}}, Scale Curve=500)
     Maximum=&FloatFromCurve(FloatCurve=DataInterface() {propertyValues {Curve {keys [<at 0 1>,<at 0.5 1>,<at 1 0>]}}}, Scale Curve=1000)
    )
   )
   &UpdateMeshOrientation(
    Rotation Rate=&FloatFromCurve(
     FloatCurve=DataInterface() {propertyValues {Curve {keys [<at 0 0>,<at 1 1>]}}}
     Scale Curve=&RandomRangeFloat(Minimum=-0.3, Maximum=-0.6)
    )
    Rotation Vector=<0 0 1>
   )
   &Drag(Drag=5, Use Linear Drag=true, Use Rotational Drag=true)
   &GravityForce(Gravity=&RandomRangeVector(
     Minimum=&MakeVector(
      X=0
      Y=0
      Z=&FloatFromCurve(
       FloatCurve=DataInterface() {propertyValues {Curve {
          defaultValue=0
          keys [
           {arriveTangent=4.4006, interpMode=RCIM_Cubic, leaveTangent=4.4006, tangentMode=RCTM_User, time=0, value=0}
           {arriveTangentWeight=0.1667, interpMode=RCIM_Cubic, tangentMode=RCTM_User, time=0.5, value=1}
          ]
         }}}
       Scale Curve=100
      )
     )
     Maximum=&MakeVector(
      X=0
      Y=0
      Z=&FloatFromCurve(
       FloatCurve=DataInterface() {propertyValues {Curve {
          defaultValue=1
          keys [
           {arriveTangent=4.4006, interpMode=RCIM_Cubic, leaveTangent=4.4006, tangentMode=RCTM_User, time=0, value=0}
           {interpMode=RCIM_Cubic, tangentMode=RCTM_User, time=0.5, value=1}
          ]
         }}}
       Scale Curve=350
      )
     )
    ))
   &SolveForcesAndVelocity(Clamp Velocity=false, Limit Acceleration=false)
  ]
  rendererClasses [&NiagaraMeshRendererProperties]
  renderers [{rendererClass=&NiagaraMeshRendererProperties, rendererIndex=0}]
 }
 NE_Rocks_Fall {
  emitterUpdateScript [
   &EmitterState(Life Cycle Mode=::NewEnumerator1, Loop Behavior=::NewEnumerator1, Loop Duration=10, UseLoopDelay=false)
   &SpawnBurst_Instantaneous(Spawn Count=15, Spawn Time=5, Use Spawn Probability=false, UseLoopCountLimit=false)
  ]
  particleSpawnScript [
   &InitializeParticle(
    Lifetime=7.2
    Mesh Scale Max=<0.005 0.005 0.01>
    Mesh Scale Min=<0.0025 0.0025 0.005>
    Mesh Scale Mode=::NewEnumerator2
    Position Offset=<0 0 50>
    Ribbon Width Mode=::NewEnumerator1
    UsePositionOffset=true
    Write Material Random=true
   )
   &InitialMeshOrientation(
    Mesh Orientation Mode=::NewEnumerator7
    Rotation=&RandomRangeVector(Minimum=<0 0 -1>, Maximum=<0 0 1>)
    Use Rotation Vector=true
   )
   &ShapeLocation(
    Large Radius=80
    MaskSpawnGroup=false
    Non Uniform Scale=&Multiply_VectorByFloat(Vector=<1 1 5>, Float=$User.Scale)
    Override Seed=false
    Shape Primitive=::NewEnumerator3
    UseOwnerScale=true
   ) {module=ShapeLocation001}
   /SetVariables(Particles.MeshIndex=&RandomRangeInt(Minimum=0, Maximum=5)) {isSetParameters=true}
  ]
  particleUpdateScript [
   &ParticleState()
   &ScaleMeshSize(Scale Factor=&VectorFromFloat(Value=&FloatFromCurve(FloatCurve=DataInterface() {propertyValues {Curve {keys [
          {interpMode=RCIM_Cubic, tangentMode=RCTM_User, time=0, value=0}
          {interpMode=RCIM_Cubic, tangentMode=RCTM_User, time=0.5, value=1}
          {arriveTangent=-0.0688, leaveTangent=-0.0688, time=0.99, value=1}
          <at 1 0>
         ]}}})))
   &VortexForce(
    Origin Pull Amount=75
    UseOriginOffset=false
    Vortex Axis Coordinate Space:ENiagaraCoordinateSpace=::Local
    Vortex Force Amount=&FloatFromCurve(
     FloatCurve=DataInterface() {propertyValues {Curve {keys [<at 0 1>,<at 0.5 1>,<at 1 0>]}}}
     Scale Curve=&RandomRangeFloat(Minimum=250, Maximum=350)
    )
   )
   &UpdateMeshOrientation(
    Rotation Rate=&FloatFromCurve(FloatCurve=DataInterface() {propertyValues {Curve {keys [<at 0 0>,<at 1 1>]}}}, Scale Curve=-0.5)
    Rotation Vector=<0 0 1>
   )
   &Drag(Drag=2.5, Use Linear Drag=true, Use Rotational Drag=true)
   &GravityForce(Gravity=&MakeVector(
     X=0
     Y=0
     Z=&FloatFromCurve(
      FloatCurve=DataInterface() {propertyValues {Curve {keys [<at 0 0.001>,<at 0.2 0>,<at 0.84 0>,<at 0.85 -0.01>,<at 0.9 0>,<at 1 1>]}}}
      Scale Curve=&RandomRangeFloat(Minimum=-25000, Maximum=-250000)
     )
    ))
   &SolveForcesAndVelocity(Clamp Velocity=false, Limit Acceleration=false)
  ]
  rendererClasses [&NiagaraMeshRendererProperties]
  renderers [{rendererClass=&NiagaraMeshRendererProperties, rendererIndex=0}]
 }
 NE_Smokes {
  simTarget=GPUComputeSim
  emitterSpawnScript [
   /SetVariables(Emitter.FakeLightActive=true) {isSetParameters=true}
   /SetVariables(Emitter.FrontEndShadowLow:NiagaraFloat=&FN_FrontEndFloatModifier(Gameplay=0.7, Front End=1.5)) {isSetParameters=true}
  ]
  emitterUpdateScript [
   &EmitterState(Life Cycle Mode=::NewEnumerator1, Loop Behavior=::NewEnumerator1, Loop Delay=1, Loop Duration=9.5, UseLoopDelay=true)
   &SpawnRate(SpawnRate=35, Use Spawn Probability=true)
   &SpawnRate(SpawnGroup=1, SpawnRate=35, Use Spawn Probability=true) {module=SpawnRate001}
  ]
  particleSpawnScript [
   &InitializeParticle(
    Color=&FNTimeOfDayColor(Night Color=<rgba 1 1 1 0.175>, Morning Color=<rgba 1 1 1 0.175>, Day Color=<rgba 1 1 1 0.075>, Evening Color=<rgba 1 1 1 0.125>)
    Lifetime Max=1.75
    Lifetime Min=1.5
    Lifetime Mode=::NewEnumerator1
    Position Offset=&MergeHelper_VectorSwitch(0=<0 0 -25>, 1=<0 0 -35>, 2=?, 3=?, Index=$Emitter.SpawnGroup, Num=::NewEnumerator0)
    Sprite Rotation Angle=&RandomRangeFloat(Minimum=0.3, Maximum=-0.3)
    Sprite Rotation Mode=::NewEnumerator1
    Sprite Size Mode=::NewEnumerator1
    Uniform Sprite Size Max=85
    Uniform Sprite Size Min=75
    UsePositionOffset=true
    Write Material Random=true
   )
   &ShapeLocation(
    Cylinder Distribution=::NewEnumerator2
    Cylinder Height=5
    Cylinder Radius=&MergeHelper_FloatSwitch(0=80, 1=30, 2=?, 3=?, Index=$Emitter.SpawnGroup, Num=::NewEnumerator0)
    MaskSpawnGroup=false
    Shape Primitive=::NewEnumerator1
    U Position=&RandomRangeFloat(Minimum=0, Maximum=1)
    UseOwnerScale=true
   )
   &CameraOffset(Camera Offset Amount=75)
   &ApplyInitialForces()
  ]
  particleUpdateScript [
   &InheritSourceMovement(
    Applied Position Scale=&Multiply_VectorByFloat(
     Vector=<1 1 1>
     Float=&Lerp_Float(
      A=&FloatFromCurve(FloatCurve=DataInterface() {propertyValues {Curve {keys [{arriveTangent=-1.0658, interpMode=RCIM_Cubic, leaveTangent=-1.0658, tangentMode=RCTM_User, time=0.1, value=1}, <at 1 0.4 RCIM_Cubic>]}}})
      B=&FloatFromCurve(FloatCurve=DataInterface() {propertyValues {Curve {keys [{arriveTangent=-0.0965, interpMode=RCIM_Cubic, leaveTangent=-0.0965, tangentMode=RCTM_User, time=0.3, value=1}, <at 1 0.7 RCIM_Cubic>]}}})
      Alpha=&MakeCustomFloatFromBool(
       Bool=&SetBoolByFloatComparison(A=&VectorLength(Vector=$Engine.Owner.Velocity), B=400)
       True Float=1
       False Float=0
      )
     )
    )
    ApplyForce=false
    ApplyPosition=true
    ApplyPositionFalloff=false
   )
   &DynamicMaterialParameters(
    Index 0 Param 1=&Add_Float(
     A=$Particles.RandomFloat
     B=&FloatFromCurve(FloatCurve=DataInterface() {propertyValues {Curve {keys [
          {arriveTangent=0.1552, interpMode=RCIM_Cubic, leaveTangent=0.1552, tangentMode=RCTM_User, time=0, value=1}
          {interpMode=RCIM_Cubic, leaveTangent=0.5473, time=1, value=1.1}
         ]}}})
    )
    Index 0 Param 4=1
    Index 2 Param 1=&FloatFromCurve(FloatCurve=DataInterface() {propertyValues {Curve {keys [
         {interpMode=RCIM_Cubic, leaveTangent=1.6214, tangentMode=RCTM_Break, time=0.1, value=0}
         {arriveTangent=0.8413, tangentMode=RCTM_User, time=1, value=1}
        ]}}})
    Index 2 Param 2=&FloatFromCurve(
     FloatCurve=DataInterface() {propertyValues {Curve {keys [{arriveTangent=0.7845, interpMode=RCIM_Cubic, leaveTangent=0.7845, tangentMode=RCTM_User, time=0, value=0.2}, <at 1 0.8 RCIM_Cubic>]}}}
     Scale Curve=&RandomRangeFloat(Minimum=0.2, Maximum=1.5)
    )
    Index 2 Param 3=&FloatFromCurve(
     FloatCurve=DataInterface() {propertyValues {Curve {keys [
         {interpMode=RCIM_Cubic, leaveTangent=1.0458, tangentMode=RCTM_Break, time=0, value=1}
         {interpMode=RCIM_Cubic, leaveTangent=11.1233, tangentMode=RCTM_Break, time=1, value=2}
        ]}}}
     Scale Curve=1.25
    )
    Index 2 Param 4=&FloatFromCurve(
     FloatCurve=DataInterface() {propertyValues {Curve {keys [<at 0 1 RCIM_Cubic>, {interpMode=RCIM_Cubic, leaveTangent=-1.2956, tangentMode=RCTM_Break, time=1, value=0.75}]}}}
     Scale Curve=1.5
    )
    Index 3 Param 1=&FloatFromCurve(
     FloatCurve=DataInterface() {propertyValues {Curve {keys [
         <at 0.5 0 RCIM_Cubic>
         {arriveTangent=1, interpMode=RCIM_Cubic, leaveTangent=1, time=1, value=1}
         {arriveTangent=-1, interpMode=RCIM_Cubic, leaveTangent=-1, time=1.5, value=1}
         {interpMode=RCIM_Cubic, leaveTangent=2, time=2, value=0}
        ]}}}
     CurveIndex=$Emitter.Age
     Scale Curve=&Multiply_Float(
      A=&FN_Select_Float_from_Time_Of_Day(Morning=3, Day=3, Evening=3, Night=3)
      B=&FN_FrontEndFloatModifier(Gameplay=1, Front End=2)
     )
    )
    Index 3 Param 2=&Multiply_Float(
     A=&Lerp_Float(
      A=33
      B=75
      Alpha=&FloatFromCurve(
       FloatCurve=DataInterface() {propertyValues {Curve {keys [
           <at 0.5 0 RCIM_Cubic>
           {arriveTangent=1, interpMode=RCIM_Cubic, leaveTangent=1, time=1, value=1}
           {arriveTangent=-1, interpMode=RCIM_Cubic, leaveTangent=-1, time=1.5, value=1}
           <at 2 0 RCIM_Cubic>
          ]}}}
       CurveIndex=$Emitter.Age
      )
     )
     B=&FN_FrontEndFloatModifier(Gameplay=1, Front End=1.25)
    )
    Module.Param0WriteEnabled=true
    Module.Param0WriteEnabled2=true
    Module.Param0WriteEnabled3=true
    Module.Param1WriteEnabled=false
    Module.Param1WriteEnabled2=true
    Module.Param1WriteEnabled3=true
    Module.Param2WriteEnabled=true
    Module.Param2WriteEnabled2=true
    Module.Param2WriteEnabled3=true
    Module.Param3WriteEnabled=true
    Module.Param3WriteEnabled2=true
    Module.Param3WriteEnabled3=true
    Write Parameter Index 2=true
    Write Parameter Index 3=true
   )
   &ParticleState()
   &ScaleSpriteSize(Uniform Curve Sprite Scale=DataInterface() {propertyValues {Curve {keys [
        {arriveTangent=1.6448, interpMode=RCIM_Cubic, leaveTangent=1.6448, tangentMode=RCTM_User, time=0, value=0.4}
        {arriveTangent=0.0224, interpMode=RCIM_Cubic, leaveTangent=0.0224, tangentMode=RCTM_User, time=1, value=1}
       ]}}})
   &ScaleColor(
    Scale Alpha=&FloatFromCurve(
     FloatCurve=DataInterface() {propertyValues {Curve {keys [
         {interpMode=RCIM_Cubic, tangentMode=RCTM_User, time=0, value=0}
         {arriveTangent=0.0617, interpMode=RCIM_Cubic, leaveTangent=0.0617, tangentMode=RCTM_User, time=0.2, value=1}
         {arriveTangent=0.0658, interpMode=RCIM_Cubic, leaveTangent=0.0658, tangentMode=RCTM_User, time=0.7, value=1}
         {tangentMode=RCTM_User, time=1, value=0}
        ]}}}
     Scale Curve=&Multiply_Float(A=1, B=1)
    )
    Scale RGB=&MakeVectorFromLinearColorRGB(Color=&MultiplyLinearColor(
      Linear Color=&RandomRangeLinearColor(Minimum=<rgba 0.5 0.5 0.5 1>, Maximum=<rgba 1 1 1 1>)
      Scale Factor=&MakeVector4_FromLinearColor(LinearColor=&LerpLinearColors(StartColor=<rgba 1 1 1 1>, EndColor=<rgba 0 0 0 1>, LerpFactor=0.1))
     ))
    ScaleA=true
    ScaleRGB=true
   )
   &VortexForce(UseOriginOffset=false, Vortex Force Amount=&Lerp_Float(A=800, B=300, Alpha=$System.Timeline002.Float))
   &Drag(Drag=10, Use Linear Drag=true, Use Rotational Drag=true)
   &AccelerationForce(Acceleration=&Multiply_VectorByFloat(Vector=<1 1 1>, Float=1))
   &SolveForcesAndVelocity(Clamp Velocity=false, Limit Acceleration=false)
  ]
  rendererClasses [&NiagaraSpriteRendererProperties]
  renderers [{rendererClass=&NiagaraSpriteRendererProperties}]
 }
 NE_Vortex {
  emitterUpdateScript [
   &EmitterState(Life Cycle Mode=::NewEnumerator1, Loop Behavior=::NewEnumerator1, Loop Delay=0.2, Loop Duration=10, UseLoopDelay=true)
   &SpawnRate(SpawnRate=3, Use Spawn Probability=false)
   &SpawnBurst_Instantaneous(Use Spawn Probability=false, UseLoopCountLimit=false)
  ]
  particleSpawnScript [
   &InitializeParticle(
    Color=&FNTimeOfDayColor(Night Color=<rgba 0.5517 0.7251 1 0.7>, Morning Color=<rgba 1 1 1 0.5>, Day Color=<rgba 1 1 1 1>, Evening Color=<rgba 1 0.8396 0.7176 1>)
    Lifetime=1.5
    Mesh Scale=&Multiply_VectorByFloat(Vector=<0.4 0.4 1>, Float=$User.Scale)
    Mesh Scale Mode=::NewEnumerator4
    Position Offset Coordinate Space=::Local
    UsePositionOffset=true
    Write Material Random=true
   )
   &InitialMeshOrientation(
    Mesh Axis To Orient=<0 0 3>
    Mesh Orientation Mode=::NewEnumerator2
    Orientation Coordinate Space=::World
    Orientation Vector=<0 0 1>
    Rotation=&RandomRangeVector(Minimum=<0 0 0>, Maximum=<0 0 1>)
    Use Rotation Vector=true
   )
  ]
  particleUpdateScript [
   &ParticleState()
   &DynamicMaterialParameters(
    Index 0 Param 1=&FloatFromCurve(FloatCurve=DataInterface() {propertyValues {Curve {keys [{interpMode=RCIM_Cubic, tangentMode=RCTM_User, time=0.5, value=0}, {interpMode=RCIM_Cubic, tangentMode=RCTM_User, time=1, value=1}]}}})
    Index 0 Param 2=&FNTimeOfDayScalar(Night Scale=0, Morning Scale=1, Evening Scale=0)
    Index 0 Param 3=1
    Module.Param0WriteEnabled=true
    Module.Param1WriteEnabled=true
    Module.Param2WriteEnabled=true
    Module.Param3WriteEnabled=true
   )
   &ScaleMeshSize(Scale Factor=&Lerp_Vector(
     A=<1 1 0.5>
     B=<5 5 1.25>
     Alpha=&FloatFromCurve(FloatCurve=DataInterface() {propertyValues {Curve {keys [{tangentMode=RCTM_User, time=0, value=1}, {tangentMode=RCTM_User, time=1, value=1.5}]}}})
    ))
   &UpdateMeshOrientation(
    Rotation Rate=&FloatFromCurve(FloatCurve=DataInterface() {propertyValues {Curve {keys [<at 0 0>,<at 1 1>]}}}, Scale Curve=0.5)
    Rotation Vector=<0 0 1>
   )
   &ScaleColor(
    Scale Alpha=&FloatFromCurve(
     FloatCurve=DataInterface() {propertyValues {Curve {keys [
         {interpMode=RCIM_Cubic, tangentMode=RCTM_User, time=0, value=0}
         {interpMode=RCIM_Cubic, tangentMode=RCTM_User, time=0.5, value=1}
         {tangentMode=RCTM_User, time=1, value=0}
        ]}}}
     Scale Curve=0.01
    )
    Scale RGB=<5 5 5>
    ScaleA=true
    ScaleRGB=true
   )
  ]
  rendererClasses [&NiagaraMeshRendererProperties]
  renderers [{rendererClass=&NiagaraMeshRendererProperties, rendererIndex=0}]
 }
}
summary {
 systemName=NS_Emote_Heaven_CloudsLoop
 emitters [
  {emitterName=NE_Rocks, rendererClasses [&NiagaraMeshRendererProperties]}
  {emitterName=NE_Rocks_Fall, rendererClasses [&NiagaraMeshRendererProperties]}
  {emitterName=NE_Vortex, rendererClasses [&NiagaraMeshRendererProperties]}
  {emitterName=NE_Smokes, simTarget=GPUComputeSim, rendererClasses [&NiagaraSpriteRendererProperties]}
  {emitterName=NE_AirFlow, rendererClasses [&NiagaraMeshRendererProperties]}
 ]
 userVariables [
  User.bIsFrontEnd {type=NiagaraBool}
  User.bIsFrontEndPreview {type=NiagaraBool}
  User.Scale {type=NiagaraFloat, defaultValue=1, description=""}
  User.VortexHeight {type=Vector3f, description="", defaultValue {struct=Vector3f, value=<0 0 1>}}
 ]
}
